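# Physics asset for the bandana face accessory of the male Teriyaki Warrior cosmetic, modelling its two dangling straps.
# Capsule chains strapA and strapB, five bones each, branch from a driver bone whose swing and twist are locked to the strap parent.
# Swing limit angles in each chain widen toward the tips.
# Unconstrained head, neck, upper spine, clavicle and left upper-arm shapes provide body collision, with no right upper-arm body.
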PhysicsAsset /BRCosmetics/Characters/Player/Male/Medium/Bodies/M_MED_Teriyaki_Warrior/Meshes/Parts/M_MED_TeriyakiWarrior_FaceAcc_Bandana_Physics_upd2
bodies {
 clavicle_l [{shapeType=Capsule, length=4.9045, radius=7.458, shapeName=clavicle_l_capsule, center=<12.2451 0 -2.1456>, rotation=<rot -0.0004 -90 100>}]
 clavicle_r [{shapeType=Capsule, length=4.0101, radius=8.0627, shapeName=clavicle_r_capsule, center=<-13.8236 0 2.8007>, rotation=<rot -0.0003 -90 100>}]
 dyn_strapA_01 [{shapeType=Capsule, length=3.5177, radius=1.0557, shapeName=dyn_strapA_01_capsule, center=<2.1395 0 0.0665>, rotation=<rot 0 -90 89.481>}]
 dyn_strapA_02 [{
   shapeType=Capsule
   length=4.0228
   radius=1.4834
   shapeName=dyn_strapA_02_capsule
   center=<3.9076 0.0948 -0.0334>
   rotation=<rot 0 -90.7378 -89.5636>
  }]
 dyn_strapA_03 [{
   shapeType=Capsule
   length=2.873
   radius=2.074
   shapeName=dyn_strapA_03_capsule
   center=<3.2281 0.0857 -0.0267>
   rotation=<rot -0.1978 92.0304 -89.2235>
  }]
 dyn_strapA_04 [{
   shapeType=Capsule
   length=3.9543
   radius=2.2202
   shapeName=dyn_strapA_04_capsule
   center=<4.1676 -0.0409 -0.0797>
   rotation=<rot -1.374 90.9482 -87.951>
  }]
 dyn_strapA_05 [{
   shapeType=Capsule
   length=3.9397
   radius=2.7866
   shapeName=dyn_strapA_05_capsule
   center=<5.2092 0.0755 -0.0369>
   rotation=<rot 0.0075 -88.1143 -89.7598>
  }]
 dyn_strapB_01 [{
   shapeType=Capsule
   length=4.0168
   radius=1.1116
   shapeName=dyn_strapB_01_capsule
   center=<3.725 0.1411 0.0092>
   rotation=<rot 0 -94.0914 -89.5493>
  }]
 dyn_strapB_02 [{
   shapeType=Capsule
   length=2.4494
   radius=1.3671
   shapeName=dyn_strapB_02_capsule
   center=<2.3843 0.0431 -0.0418>
   rotation=<rot -0.2693 89.5315 -88.4698>
  }]
 dyn_strapB_03 [{
   shapeType=Capsule
   length=2.9354
   radius=1.6476
   shapeName=dyn_strapB_03_capsule
   center=<3.0692 0.0558 -0.0519>
   rotation=<rot -0.584 90.8818 -88.1231>
  }]
 dyn_strapB_04 [{
   shapeType=Capsule
   length=3.4976
   radius=1.6666
   shapeName=dyn_strapB_04_capsule
   center=<3.7515 0.033 -0.0289>
   rotation=<rot -0.028 89.8091 -89.5725>
  }]
 dyn_strapB_05 [{
   shapeType=Capsule
   length=3.137
   radius=1.8647
   shapeName=dyn_strapB_05_capsule
   center=<3.2298 0.0625 -0.0372>
   rotation=<rot -0.0283 88.9587 -89.5815>
  }]
 dyn_strap_driver [{shapeType=Capsule, length=0.505, radius=0.505, shapeName=dyn_strap_driver_capsule, center=<0 -0.4135 -0.259>, rotation=<rot 39.9999 -90 0>}]
 dyn_strap_parent [{shapeType=Sphere, radius=1.8347, shapeName=dyn_strap_parent_sphere, center=<0.5093 -0.5831 -1.0496>}]
 head [
  {shapeType=Sphere, radius=13.3798, shapeName=head_sphere, center=<3.4143 5.3751 0.1786>}
  {shapeType=Capsule, length=4.9806, radius=15, shapeName=head_capsule, center=<-7.5732 4.8086 0>, rotation=<rot 90 180 150.0003>}
 ]
 neck_01 [{shapeType=Capsule, length=12.5911, radius=12.2506, shapeName=neck_01_capsule, center=<0.4738 3.6986 0>, rotation=<rot 0 -90.0005 90>}]
 spine_04 [{shapeType=Capsule, length=14.6495, radius=10, shapeName=spine_04_capsule, center=<4.0938 -3.2637 0>, rotation=<rot 0 -90 0>}]
 spine_05 [{shapeType=Capsule, length=15.9556, radius=12.5739, shapeName=spine_05_capsule, center=<0.0281 0.0804 0>, rotation=<rot 0 -90 0>}]
 upperarm_l [{
   shapeType=Capsule
   length=17.49
   radius=5.981
   shapeName=upperarm_l_capsule
   center=<9.0755 0 -0.2416>
   rotation=<rot -0.0003 -89.9997 86.6081>
  }]
}
constraints [
 {bone1Name=dyn_strapA_02, bone2Name=dyn_strapA_01, swing1LimitDegrees=35, swing2LimitDegrees=35, twistLimitDegrees=15}
 {bone1Name=dyn_strapA_03, bone2Name=dyn_strapA_02, swing1LimitDegrees=35, swing2LimitDegrees=35, twistLimitDegrees=15}
 {bone1Name=dyn_strapA_04, bone2Name=dyn_strapA_03, swing1LimitDegrees=45, swing2LimitDegrees=45, twistLimitDegrees=15}
 {bone1Name=dyn_strapA_05, bone2Name=dyn_strapA_04, swing1LimitDegrees=45, swing2LimitDegrees=45, twistLimitDegrees=15}
 {bone1Name=dyn_strapB_02, bone2Name=dyn_strapB_01, swing1LimitDegrees=35, swing2LimitDegrees=35, twistLimitDegrees=15}
 {bone1Name=dyn_strapB_03, bone2Name=dyn_strapB_02, swing1LimitDegrees=35, swing2LimitDegrees=35, twistLimitDegrees=15}
 {bone1Name=dyn_strapB_04, bone2Name=dyn_strapB_03, swing1LimitDegrees=45, swing2LimitDegrees=45, twistLimitDegrees=15}
 {bone1Name=dyn_strapB_05, bone2Name=dyn_strapB_04, swing1LimitDegrees=45, swing2LimitDegrees=45, twistLimitDegrees=15}
 {bone1Name=dyn_strapA_01, bone2Name=dyn_strap_driver, swing1LimitDegrees=25, swing2LimitDegrees=25, twistLimitDegrees=10}
 {bone1Name=dyn_strapB_01, bone2Name=dyn_strap_driver, swing1LimitDegrees=25, swing2LimitDegrees=25, twistLimitDegrees=10}
 {
  bone1Name=dyn_strap_driver
  bone2Name=dyn_strap_parent
  swing1LimitDegrees=35
  swing1Motion=Locked
  swing2LimitDegrees=35
  swing2Motion=Locked
  twistLimitDegrees=15
  twistMotion=Locked
 }
]
