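# Physics asset for the bandana face accessory on the medium male Teriyaki Warrior player body, set up for the physics motion of its two straps.
# Strap A and strap B are five-box chains with limited swing and twist, both constrained to the strap driver bone.
# An extra constraint links the first segments of the two straps.
# Capsules on the head, neck, upper spine, clavicles and left upper arm approximate the body around the straps.
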PhysicsAsset /Game/Characters/Player/Male/Medium/Bodies/M_MED_Teriyaki_Warrior/Meshes/Parts/M_MED_TeriyakiWarrior_FaceAcc_Bandana_Physics_upd
bodies {
 clavicle_l [{shapeType=Capsule, length=9.9528, radius=7.7101, center=<9.6425 -0.8647 -1.1878>, rotation=<rot 69.9994 0 0>}]
 clavicle_r [{shapeType=Capsule, length=13.7488, radius=5.6376, center=<-6.1192 0.0002 0>, rotation=<rot 90 0 0>}]
 dyn_strapA_01 [{shapeType=Box, extentX=4.8673, extentY=1.5107, extentZ=2.0354, center=<2.132 0 0.0002>, rotation=<rot 0 0 90>}]
 dyn_strapA_02 [{shapeType=Box, extentX=8.2557, extentY=1.4925, extentZ=3.1884, center=<3.737 0 0.0001>, rotation=<rot 0 0 90>}]
 dyn_strapA_03 [{shapeType=Box, extentX=7.2881, extentY=1.3265, extentZ=3.8189, center=<2.9868 0 0.0001>, rotation=<rot 0 0 90>}]
 dyn_strapA_04 [{shapeType=Box, extentX=7.0813, extentY=1.4724, extentZ=4.6312, center=<3.5964 0 -0.0002>, rotation=<rot 0 0 90>}]
 dyn_strapA_05 [{shapeType=Box, extentX=8.6124, extentY=1.5546, extentZ=5.0488, center=<4.8912 0 -0.0001>, rotation=<rot 0 0 90>}]
 dyn_strapB_01 [{shapeType=Box, extentX=6.9092, extentY=1.2093, extentZ=2.2938, center=<3.5756 0 -0.0006>, rotation=<rot 0.0001 -0.0001 105.1429>}]
 dyn_strapB_02 [{shapeType=Box, extentX=4.8925, extentY=1.4002, extentZ=2.3206, center=<2.2163 0 -0.0002>, rotation=<rot 0 0 90>}]
 dyn_strapB_03 [{shapeType=Box, extentX=5.0294, extentY=1.5173, extentZ=2.7, center=<2.0747 0 0.0003>, rotation=<rot 0 0 90>}]
 dyn_strapB_04 [{shapeType=Box, extentX=7.6028, extentY=1.37, extentZ=3.3657, center=<3.3705 0 0.0001>, rotation=<rot 0 0 90>}]
 dyn_strapB_05 [{shapeType=Box, extentX=6.6505, extentY=1.4242, extentZ=3.41, center=<3.2158 0 -0.0005>, rotation=<rot 0 0 90>}]
 dyn_strap_driver [{shapeType=Box, extentX=1.01, extentY=1.01, extentZ=1.01, rotation=<rot 0 -90 0>}]
 dyn_strap_parent [{shapeType=Box, extentX=1.01, extentY=1.01, extentZ=1.01, rotation=<rot 0 -90 0>}]
 head [{shapeType=Capsule, length=6.2808, radius=12.0939, center=<0.92 3.4471 0>, rotation=<rot 90 0 0>}]
 neck_01 [{shapeType=Capsule, length=0.7305, radius=10.2992, center=<0 3.8782 0>, rotation=<rot 90 0 0>}]
 spine_03 [{shapeType=Capsule, length=20.9157, radius=8.285, center=<0 -1.8097 -0.0001>}]
 spine_04 [
  {shapeType=Capsule, length=18.7173, radius=10.4834, center=<4.637 -1.8028 -0.0001>}
  {shapeType=Capsule, length=20.1277, radius=9.6223, center=<5.1168 -5.1732 0>}
 ]
 spine_05 [
  {shapeType=Capsule, length=18.8328, radius=10.3679, center=<-1.6847 -2.9 0.0001>}
  {shapeType=Capsule, length=18.6213, radius=7.8786, center=<4.2601 -4.6728 0.4988>}
 ]
 upperarm_l [{
   shapeType=Capsule
   length=19.2059
   radius=6.0017
   shapeName=upperarm_l_capsule
   center=<11.4338 -0.5849 -0.8841>
   rotation=<rot -0.0004 -84.6104 91.2575>
  }]
}
constraints [
 {bone1Name=dyn_strapA_01, bone2Name=dyn_strap_driver, swing1LimitDegrees=45, swing2LimitDegrees=35, twistLimitDegrees=16}
 {bone1Name=dyn_strapA_02, bone2Name=dyn_strapA_01, swing1LimitDegrees=45, swing2LimitDegrees=35, twistLimitDegrees=16}
 {bone1Name=dyn_strapA_03, bone2Name=dyn_strapA_02, swing1LimitDegrees=35, swing2LimitDegrees=35, twistLimitDegrees=16}
 {bone1Name=dyn_strapA_04, bone2Name=dyn_strapA_03, swing1LimitDegrees=35, swing2LimitDegrees=35, twistLimitDegrees=16}
 {bone1Name=dyn_strapA_05, bone2Name=dyn_strapA_04, swing1LimitDegrees=45, swing2LimitDegrees=35, twistLimitDegrees=16}
 {bone1Name=dyn_strapB_01, bone2Name=dyn_strap_driver, swing1LimitDegrees=45, swing2LimitDegrees=35, twistLimitDegrees=16}
 {bone1Name=dyn_strapB_02, bone2Name=dyn_strapB_01, swing1LimitDegrees=45, swing2LimitDegrees=35, twistLimitDegrees=16}
 {bone1Name=dyn_strapB_03, bone2Name=dyn_strapB_02, swing1LimitDegrees=35, swing2LimitDegrees=35, twistLimitDegrees=16}
 {bone1Name=dyn_strapB_04, bone2Name=dyn_strapB_03, swing1LimitDegrees=35, swing2LimitDegrees=35, twistLimitDegrees=16}
 {bone1Name=dyn_strapB_05, bone2Name=dyn_strapB_04, swing1LimitDegrees=45, swing2LimitDegrees=35, twistLimitDegrees=16}
 {bone1Name=dyn_strapB_01, bone2Name=dyn_strapA_01, swing1LimitDegrees=35, swing2LimitDegrees=35, twistLimitDegrees=16}
]
