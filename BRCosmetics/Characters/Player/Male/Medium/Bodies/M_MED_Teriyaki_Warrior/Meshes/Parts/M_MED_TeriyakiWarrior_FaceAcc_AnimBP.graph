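# Animation Blueprint for the Teriyaki Warrior outfit's face accessory, adding secondary motion to its two dangling straps.
# A cached pose blends parent and Body mesh poses, with the strap root constrained to neck and head.
# It feeds a Blend by int of RigidBody, limited AnimDynamics chains, and LookAt plus Trail controllers.
# Update sets gravity and SimType from bRigidBodyEnabled; older speed-driven strap logic and extra physics nodes are left unwired.
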AnimBlueprint /BRCosmetics/Characters/Player/Male/Medium/Bodies/M_MED_Teriyaki_Warrior/Meshes/Parts/M_MED_TeriyakiWarrior_FaceAcc_AnimBP
target_skeleton=&/M_MED_TeriyakiWarrior_FaceAcc_Skeleton
graphs [
 AnimGraph {nodes [
   n1 = Get CurrentLOD()
   n2 = Get bRigidBodyEnabled()
   Blend Poses by bool(bActiveValue=OR Boolean(A=NOT Boolean(A=@n2), B="integer > integer"(A=@n1, B=1) {is=title K2Node_CallFunction})) {blendType=Linear}
   Component To Local()
   Component To Local()
   "Constraint - dyn_strap_driver (P)"(Alpha=1, ConstraintWeights_0=1) {
    alphaBoolBlend {blendOption=Linear}
    boneToModify {boneName=dyn_strap_driver}
    constraintSetup [{targetBone {boneName=head}}]
    constraintWeights [1]
   }
   Local To Component()
   n3 = Copy Bone - Source Bone: dyn_strapB_08 - Target Bone: spine_05!(ComponentPose:Component Space Pose Link Structure, Alpha:float=1, bCopyTranslation:bool=true, bCopyRotation:bool=true, bCopyScale:bool=False) -> (Pose:Component Space Pose Link Structure) {
    is=title AnimGraphNode_CopyBone
    controlSpace=BCS_BoneSpace
    alphaBoolBlend {blendOption=Linear}
    sourceBone {boneName=dyn_strapB_08}
    targetBone {boneName=spine_05}
   }
   n4 = Copy Bone - Source Bone: dyn_strapA_09 - Target Bone: spine_05!(
    ComponentPose:Component Space Pose Link Structure=@n3
    Alpha:float=1
    bCopyTranslation:bool=true
    bCopyRotation:bool=true
    bCopyScale:bool=False
   ) -> (Pose:Component Space Pose Link Structure) {
    is=title AnimGraphNode_CopyBone
    controlSpace=BCS_BoneSpace
    alphaBoolBlend {blendOption=Linear}
    sourceBone {boneName=dyn_strapA_09}
    targetBone {boneName=spine_05}
   }
   n5 = "Transform (Modify) Bone - Bone: dyn_strapB_07"(ComponentPose=@n4, Alpha=1, Translation="0,0,0", Rotation="0.000,0.000,-25", Scale="1,1,1") {rotationMode=BMM_Additive, alphaBoolBlend {blendOption=Linear}, boneToModify {boneName=dyn_strapB_07}, rotation {roll=-25}}
   "Transform (Modify) Bone - Bone: dyn_strapA_08"(ComponentPose=@n5, Alpha=1, Translation="0,0,0", Rotation="0.000,0.000,-25", Scale="1,1,1") {rotationMode=BMM_Additive, alphaBoolBlend {blendOption=Linear}, boneToModify {boneName=dyn_strapA_08}, rotation {roll=-25}}
   n6 = "Transform (Modify) Bone - Bone: dyn_strap_parent"(Alpha=Get straps_stand_rot() -> (straps_stand_rot:double), Translation="0,0,0", Rotation="0,0,-25", Scale="1,1,1") {
    rotationMode=BMM_Additive
    rotationSpace=BCS_BoneSpace
    alphaBoolBlend {blendOption=Linear}
    boneToModify {boneName=dyn_strap_parent}
    rotation {roll=-25}
   }
   n7 = RigidBody(ComponentPose=@n6) {
    bClampLinearTranslationLimitToRefPose=true
    bOverrideWorldGravity=true
    overridePhysicsAsset=/Game/Characters/Player/Male/Medium/Bodies/M_MED_Teriyaki_Warrior/Meshes/Parts/M_MED_TeriyakiWarrior_FaceAcc_Bandana_Physics_upd
    simulationSpace=BaseBoneSpace
    alphaBoolBlend {blendOption=Linear}
    baseBoneRef {boneName=root}
    overrideWorldGravity {z=-980}
    simSpaceSettings {
     maxAngularAcceleration=500
     maxAngularVelocity=3
     maxLinearAcceleration=400
     maxLinearVelocity=250
     velocityScaleZ=0
     worldAlpha=1
     externalLinearDragV=<0.1 0.1>
    }
   }
   "Transform (Modify) Bone - Bone: dyn_strap_parent"(ComponentPose=@n7, Alpha=Get strap_rot_alpha() -> (strap_rot_alpha:double), Translation="0,0,0", Rotation="0.000,0.000,25", Scale="1,1,1") {
    rotationMode=BMM_Additive
    rotationSpace=BCS_BoneSpace
    alphaBoolBlend {blendOption=Linear}
    alphaScaleBiasClamp {bInterpResult=true, interpSpeedDecreasing=5, interpSpeedIncreasing=5}
    boneToModify {boneName=dyn_strap_parent}
    rotation {roll=25}
   }
   RigidBody(Alpha=0.6) {
    bClampLinearTranslationLimitToRefPose=true
    bOverrideWorldGravity=true
    overridePhysicsAsset=/Game/Accessories/Hats/F_MED_Skull_Brite_Hat/Meshes/F_MED_Skull_Brite_HairTips_Physics
    simulationSpace=BaseBoneSpace
    alphaBoolBlend {blendOption=Linear}
    baseBoneRef {boneName=root}
    componentAppliedLinearAccClamp=<3000 3000 3000>
    componentLinearAccScale=<0.5 0.5 0.5>
    componentLinearVelScale=<0.5 0.5 0.5>
    overrideWorldGravity {z=-980}
   }
   n8 = Get straps_vector() -> (straps_vector:Vector)
   RigidBody!(
    ComponentPose
    Alpha=Get straps_standing_alpha() -> (straps_standing_alpha:double)
    ExternalForce=Get straps_force_vector() -> (straps_force_vector:Vector)
    ComponentLinearAccScale:Vector=@n8
    ComponentLinearVelScale:Vector=@n8
   ) -> (Pose) {
    bClampLinearTranslationLimitToRefPose=true
    bOverrideWorldGravity=true
    overridePhysicsAsset=/Game/Characters/Player/Male/Medium/Bodies/M_MED_Teriyaki_Warrior/Meshes/Parts/M_MED_TeriyakiWarrior_FaceAcc_Bandana_Physics
    simulationSpace=BaseBoneSpace
    alphaBoolBlend {blendOption=Linear}
    alphaScaleBiasClamp {bInterpResult=true}
    baseBoneRef {boneName=root}
    componentAppliedLinearAccClamp=<3000 1500 2000>
    componentLinearAccScale {z=0.5}
    componentLinearVelScale {z=0.5}
    overrideWorldGravity {z=-980}
   }
   n9 = "Transform (Modify) Bone - Bone: dyn_strap_parent"!(ComponentPose=Local To Component(LocalPose=Use cached pose 'AnimCache'()), Rotation="4.925,-16.465,-12") -> (Pose) {
    rotationMode=BMM_Additive
    rotationSpace=BCS_BoneSpace
    alphaBoolBlend {blendOption=Linear}
    boneToModify {boneName=dyn_strap_parent}
    rotation=<rot 4.925 -16.465 -12>
   }
   n10 = Look At - Bone: dyn_strapB_02(ComponentPose=@n9, Alpha=1) {
    alphaBoolBlend {blendOption=Linear}
    boneToModify {boneName=dyn_strapB_02}
    lookAtLocation=<-5 -15 -5>
    lookAtTarget {boneReference {boneName=spine_05}}
    lookAt_Axis {axis=<1 0>}
    lookUp_Axis {axis=<0 1>}
   }
   n11 = Look At - Bone: dyn_strapA_02(ComponentPose=@n10, Alpha=1) {
    alphaBoolBlend {blendOption=Linear}
    boneToModify {boneName=dyn_strapA_02}
    lookAtLocation=<-5 -14 2>
    lookAtTarget {boneReference {boneName=spine_05}}
    lookAt_Axis {axis=<1 0>}
    lookUp_Axis {axis=<0 1>}
   }
   n12 = Trail controller - Bone: dyn_strapA_05(ComponentPose=@n11, Alpha=1) {
    bUsePlanarLimit=true
    chainLength=3
    debugLifeTime=0.005
    maxDeltaTime=0.0333
    relaxationSpeedScale=3
    planarLimits [
     {drivingBone {boneName=spine_05}, planeTransform {location=<0 -10 0>, rotation=<rot 0 0 -90>, scale=<1 1 1>}}
     {drivingBone {boneName=spine_05}, planeTransform {location=<10 -10 0>, rotation=<rot 0 24.9999 -90>, scale=<1 1 1>}}
    ]
    rotationLimits [{limitMax=<180 180 180>}, {limitMax=<180 180 180>}, {limitMax=<180 180 180>}]
    rotationOffsets [<0 0 0>, <0 0 0>, <0 0 0>]
    trailBone {boneName=dyn_strapA_05}
   }
   n13 = Trail controller - Bone: dyn_strapB_05(ComponentPose=@n12, Alpha=1) {
    bUsePlanarLimit=true
    chainLength=3
    maxDeltaTime=0.0333
    relaxationSpeedScale=3
    planarLimits [
     {drivingBone {boneName=spine_05}, planeTransform {location=<0 -10 0>, rotation=<rot 0 0 -90>, scale=<1 1 1>}}
     {drivingBone {boneName=spine_05}, planeTransform {location=<10 -10 0>, rotation=<rot 0 24.9999 -90>, scale=<1 1 1>}}
    ]
    rotationLimits [{limitMax=<180 180 180>}, {limitMax=<180 180 180>}, {limitMax=<180 180 180>}]
    rotationOffsets [<0 0 0>, <0 0 0>, <0 0 0>]
    trailBone {boneName=dyn_strapB_05}
   }
   n14 = Anim Dynamics - Bone: dyn_strapA_01!(ComponentPose=Local To Component(LocalPose=Use cached pose 'AnimCache'()), Alpha=1, GravityOverride:Vector="0,0,-980", ExternalForce="0,0,0") -> (Pose) {
    angularDampingOverride=0.999
    angularSpringConstant=50
    bAngularSpring=true
    bOverrideAngularDamping=true
    bUseGravityOverride=true
    bUsePlanarLimit=false
    bUseSphericalLimits=true
    simulationSpace=RootRelative
    alphaBoolBlend {blendOption=Linear}
    boundBone {boneName=dyn_strapA_01}
    componentAppliedLinearAccClamp=<10000 10000 10000>
    componentLinearAccScale=<1 1 1>
    componentLinearVelScale=<1 1 1>
    gravityOverride {z=-980}
    physicsBodyDefinitions [{
      boundBone {boneName=dyn_strapA_01}
      boxExtents=<4.5 4.5 4.5>
      constraintSetup {angularLimitsMax=<5 25 25>, angularLimitsMin=<-5 -25 -25>, angularTarget=<1 0 0>}
      localJointOffset=<0.2 0 0>
     }]
    relativeSpaceBone {boneName=head}
    sphericalLimits [{limitRadius=12, drivingBone {boneName=head}, sphereLocalOffset=<0 5 0>}]
   }
   n15 = Anim Dynamics - Bone: dyn_strapA_02!(ComponentPose=@n14, Alpha=1, GravityOverride:Vector="0,0,-980", ExternalForce="0,0,0") -> (Pose) {
    angularDampingOverride=0.999
    angularSpringConstant=50
    bAngularSpring=true
    bOverrideAngularDamping=true
    bUseGravityOverride=true
    bUseSphericalLimits=true
    numSolverIterationsPostUpdate=3
    numSolverIterationsPreUpdate=6
    simulationSpace=RootRelative
    alphaBoolBlend {blendOption=Linear}
    boundBone {boneName=dyn_strapA_02}
    componentAppliedLinearAccClamp=<10000 10000 10000>
    componentLinearAccScale=<1 1 1>
    componentLinearVelScale=<1 1 1>
    gravityOverride {z=-980}
    physicsBodyDefinitions [{
      boundBone {boneName=dyn_strapA_02}
      boxExtents=<4.5 4.5 4.5>
      constraintSetup {angularLimitsMax=<5 25 25>, angularLimitsMin=<-5 -25 -25>, angularTarget=<1 0 0>}
      localJointOffset=<0.2 0 0>
     }]
    planarLimits [{drivingBone {boneName=spine_04}, planeTransform {location=<7 -12 0>, rotation=<rot 0 40 -90>}}]
    relativeSpaceBone {boneName=head}
    sphericalLimits [
     {limitRadius=14, drivingBone {boneName=head}, sphereLocalOffset=<0 5 0>}
     {limitRadius=11, drivingBone {boneName=clavicle_l}, sphereLocalOffset=<5 0 0>}
     {limitRadius=11, drivingBone {boneName=clavicle_l}, sphereLocalOffset=<15 0 0>}
     {limitRadius=11, drivingBone {boneName=clavicle_r}, sphereLocalOffset=<-5 0 0>}
     {limitRadius=11, drivingBone {boneName=clavicle_r}, sphereLocalOffset=<-15 0 0>}
     {limitRadius=11, drivingBone {boneName=spine_05}, sphereLocalOffset=<7 0 0>}
    ]
   }
   n16 = Anim Dynamics - Bone: dyn_strapA_03!(ComponentPose=@n15, Alpha=1, GravityOverride:Vector="0,0,-980", ExternalForce="0,0,0") -> (Pose) {
    angularDampingOverride=0.999
    angularSpringConstant=50
    bAngularSpring=true
    bOverrideAngularDamping=true
    bUseGravityOverride=true
    bUseSphericalLimits=true
    numSolverIterationsPostUpdate=3
    numSolverIterationsPreUpdate=6
    simulationSpace=RootRelative
    alphaBoolBlend {blendOption=Linear}
    boundBone {boneName=dyn_strapA_03}
    componentAppliedLinearAccClamp=<10000 10000 10000>
    componentLinearAccScale=<1 1 1>
    componentLinearVelScale=<1 1 1>
    gravityOverride {z=-980}
    physicsBodyDefinitions [{
      boundBone {boneName=dyn_strapA_03}
      boxExtents=<4.5 4.5 4.5>
      constraintSetup {angularLimitsMax=<5 25 25>, angularLimitsMin=<-5 -25 -25>, angularTarget=<1 0 0>}
      localJointOffset=<0.2 0 0>
     }]
    planarLimits [{drivingBone {boneName=spine_04}, planeTransform {location=<7 -12 0>, rotation=<rot 0 40 -90>}}]
    relativeSpaceBone {boneName=head}
    sphericalLimits [
     {limitRadius=14, drivingBone {boneName=head}, sphereLocalOffset=<0 5 0>}
     {limitRadius=11, drivingBone {boneName=clavicle_l}, sphereLocalOffset=<5 0 0>}
     {limitRadius=11, drivingBone {boneName=clavicle_l}, sphereLocalOffset=<15 0 0>}
     {limitRadius=11, drivingBone {boneName=clavicle_r}, sphereLocalOffset=<-5 0 0>}
     {limitRadius=11, drivingBone {boneName=clavicle_r}, sphereLocalOffset=<-15 0 0>}
     {limitRadius=11, drivingBone {boneName=spine_05}, sphereLocalOffset=<7 0 0>}
    ]
   }
   n17 = Anim Dynamics - Bone: dyn_strapA_04!(ComponentPose=@n16, Alpha=1, GravityOverride:Vector="0,0,-980", ExternalForce="0,0,0") -> (Pose) {
    angularDampingOverride=0.999
    angularSpringConstant=50
    bAngularSpring=true
    bOverrideAngularDamping=true
    bUseGravityOverride=true
    bUseSphericalLimits=true
    numSolverIterationsPostUpdate=3
    numSolverIterationsPreUpdate=6
    simulationSpace=RootRelative
    alphaBoolBlend {blendOption=Linear}
    boundBone {boneName=dyn_strapA_04}
    componentAppliedLinearAccClamp=<10000 10000 10000>
    componentLinearAccScale=<1 1 1>
    componentLinearVelScale=<1 1 1>
    gravityOverride {z=-980}
    physicsBodyDefinitions [{
      boundBone {boneName=dyn_strapA_04}
      boxExtents=<4.5 4.5 4.5>
      constraintSetup {angularLimitsMax=<5 25 25>, angularLimitsMin=<-5 -25 -25>, angularTarget=<1 0 0>}
      localJointOffset=<0.2 0 0>
     }]
    planarLimits [{drivingBone {boneName=spine_04}, planeTransform {location=<7 -12 0>, rotation=<rot 0 40 -90>}}]
    relativeSpaceBone {boneName=head}
    sphericalLimits [
     {limitRadius=14, drivingBone {boneName=head}, sphereLocalOffset=<0 5 0>}
     {limitRadius=11, drivingBone {boneName=clavicle_l}, sphereLocalOffset=<5 0 0>}
     {limitRadius=11, drivingBone {boneName=clavicle_l}, sphereLocalOffset=<15 0 0>}
     {limitRadius=11, drivingBone {boneName=clavicle_r}, sphereLocalOffset=<-5 0 0>}
     {limitRadius=11, drivingBone {boneName=clavicle_r}, sphereLocalOffset=<-15 0 0>}
     {limitRadius=11, drivingBone {boneName=spine_05}, sphereLocalOffset=<7 0 0>}
    ]
   }
   n18 = Anim Dynamics - Bone: dyn_strapA_05!(ComponentPose=@n17, Alpha=1, GravityOverride:Vector="0,0,-980", ExternalForce="0,0,0") -> (Pose) {
    angularDampingOverride=0.999
    angularSpringConstant=50
    bAngularSpring=true
    bOverrideAngularDamping=true
    bUseGravityOverride=true
    bUseSphericalLimits=true
    numSolverIterationsPostUpdate=3
    numSolverIterationsPreUpdate=6
    simulationSpace=RootRelative
    alphaBoolBlend {blendOption=Linear}
    boundBone {boneName=dyn_strapA_05}
    componentAppliedLinearAccClamp=<10000 10000 10000>
    componentLinearAccScale=<1 1 1>
    componentLinearVelScale=<1 1 1>
    gravityOverride {z=-980}
    physicsBodyDefinitions [{
      boundBone {boneName=dyn_strapA_05}
      boxExtents=<4.5 4.5 4.5>
      constraintSetup {angularLimitsMax=<5 25 25>, angularLimitsMin=<-5 -25 -25>, angularTarget=<1 0 0>}
      localJointOffset=<0.2 0 0>
     }]
    planarLimits [{drivingBone {boneName=spine_04}, planeTransform {location=<7 -12 0>, rotation=<rot 0 40 -90>}}]
    relativeSpaceBone {boneName=head}
    sphericalLimits [
     {limitRadius=14, drivingBone {boneName=head}, sphereLocalOffset=<0 5 0>}
     {limitRadius=11, drivingBone {boneName=clavicle_l}, sphereLocalOffset=<5 0 0>}
     {limitRadius=11, drivingBone {boneName=clavicle_l}, sphereLocalOffset=<15 0 0>}
     {limitRadius=11, drivingBone {boneName=clavicle_r}, sphereLocalOffset=<-5 0 0>}
     {limitRadius=11, drivingBone {boneName=clavicle_r}, sphereLocalOffset=<-15 0 0>}
     {limitRadius=11, drivingBone {boneName=spine_05}, sphereLocalOffset=<7 0 0>}
    ]
   }
   n19 = Anim Dynamics - Bone: dyn_strapB_01!(ComponentPose=@n18, Alpha=1, GravityOverride:Vector="0,0,-980", ExternalForce="0,0,0") -> (Pose) {
    angularDampingOverride=0.999
    angularSpringConstant=50
    bAngularSpring=true
    bOverrideAngularDamping=true
    bUseGravityOverride=true
    bUsePlanarLimit=false
    bUseSphericalLimits=true
    simulationSpace=RootRelative
    alphaBoolBlend {blendOption=Linear}
    boundBone {boneName=dyn_strapB_01}
    componentAppliedLinearAccClamp=<10000 10000 10000>
    componentLinearAccScale=<1 1 1>
    componentLinearVelScale=<1 1 1>
    gravityOverride {z=-980}
    physicsBodyDefinitions [{
      boundBone {boneName=dyn_strapB_01}
      boxExtents=<4.5 4.5 4.5>
      constraintSetup {angularLimitsMax=<5 25 25>, angularLimitsMin=<-5 -25 -25>, angularTarget=<1 0 0>}
      localJointOffset=<0.2 0 0>
     }]
    relativeSpaceBone {boneName=head}
    sphericalLimits [{limitRadius=12, drivingBone {boneName=head}, sphereLocalOffset=<0 5 0>}]
   }
   n20 = Anim Dynamics - Bone: dyn_strapB_02!(ComponentPose=@n19, Alpha=1, GravityOverride:Vector="0,0,-980", ExternalForce="0,0,0") -> (Pose) {
    angularDampingOverride=0.999
    angularSpringConstant=50
    bAngularSpring=true
    bOverrideAngularDamping=true
    bUseGravityOverride=true
    bUseSphericalLimits=true
    numSolverIterationsPostUpdate=3
    numSolverIterationsPreUpdate=6
    simulationSpace=RootRelative
    alphaBoolBlend {blendOption=Linear}
    boundBone {boneName=dyn_strapB_02}
    componentAppliedLinearAccClamp=<10000 10000 10000>
    componentLinearAccScale=<1 1 1>
    componentLinearVelScale=<1 1 1>
    gravityOverride {z=-980}
    physicsBodyDefinitions [{
      boundBone {boneName=dyn_strapB_02}
      boxExtents=<4.5 4.5 4.5>
      constraintSetup {angularLimitsMax=<5 25 25>, angularLimitsMin=<-5 -25 -25>, angularTarget=<1 0 0>}
      localJointOffset=<0.2 0 0>
     }]
    planarLimits [{drivingBone {boneName=spine_04}, planeTransform {location=<7 -12 0>, rotation=<rot 0 40 -90>}}]
    relativeSpaceBone {boneName=head}
    sphericalLimits [
     {limitRadius=14, drivingBone {boneName=head}, sphereLocalOffset=<0 5 0>}
     {limitRadius=11, drivingBone {boneName=clavicle_l}, sphereLocalOffset=<5 0 0>}
     {limitRadius=11, drivingBone {boneName=clavicle_l}, sphereLocalOffset=<15 0 0>}
     {limitRadius=11, drivingBone {boneName=clavicle_r}, sphereLocalOffset=<-5 0 0>}
     {limitRadius=11, drivingBone {boneName=clavicle_r}, sphereLocalOffset=<-15 0 0>}
     {limitRadius=11, drivingBone {boneName=spine_05}, sphereLocalOffset=<7 0 0>}
    ]
   }
   n21 = Anim Dynamics - Bone: dyn_strapB_03!(ComponentPose=@n20, Alpha=1, GravityOverride:Vector="0,0,-980", ExternalForce="0,0,0") -> (Pose) {
    angularDampingOverride=0.999
    angularSpringConstant=50
    bAngularSpring=true
    bOverrideAngularDamping=true
    bUseGravityOverride=true
    bUseSphericalLimits=true
    numSolverIterationsPostUpdate=3
    numSolverIterationsPreUpdate=6
    simulationSpace=RootRelative
    alphaBoolBlend {blendOption=Linear}
    boundBone {boneName=dyn_strapB_03}
    componentAppliedLinearAccClamp=<10000 10000 10000>
    componentLinearAccScale=<1 1 1>
    componentLinearVelScale=<1 1 1>
    gravityOverride {z=-980}
    physicsBodyDefinitions [{
      boundBone {boneName=dyn_strapB_03}
      boxExtents=<4.5 4.5 4.5>
      constraintSetup {angularLimitsMax=<5 25 25>, angularLimitsMin=<-5 -25 -25>, angularTarget=<1 0 0>}
      localJointOffset=<0.2 0 0>
     }]
    planarLimits [{drivingBone {boneName=spine_04}, planeTransform {location=<7 -12 0>, rotation=<rot 0 40 -90>}}]
    relativeSpaceBone {boneName=head}
    sphericalLimits [
     {limitRadius=14, drivingBone {boneName=head}, sphereLocalOffset=<0 5 0>}
     {limitRadius=11, drivingBone {boneName=clavicle_l}, sphereLocalOffset=<5 0 0>}
     {limitRadius=11, drivingBone {boneName=clavicle_l}, sphereLocalOffset=<15 0 0>}
     {limitRadius=11, drivingBone {boneName=clavicle_r}, sphereLocalOffset=<-5 0 0>}
     {limitRadius=11, drivingBone {boneName=clavicle_r}, sphereLocalOffset=<-15 0 0>}
     {limitRadius=11, drivingBone {boneName=spine_05}, sphereLocalOffset=<7 0 0>}
    ]
   }
   n22 = Anim Dynamics - Bone: dyn_strapB_04!(ComponentPose=@n21, Alpha=1, GravityOverride:Vector="0,0,-980", ExternalForce="0,0,0") -> (Pose) {
    angularDampingOverride=0.999
    angularSpringConstant=50
    bAngularSpring=true
    bOverrideAngularDamping=true
    bUseGravityOverride=true
    bUseSphericalLimits=true
    numSolverIterationsPostUpdate=3
    numSolverIterationsPreUpdate=6
    simulationSpace=RootRelative
    alphaBoolBlend {blendOption=Linear}
    boundBone {boneName=dyn_strapB_04}
    componentAppliedLinearAccClamp=<10000 10000 10000>
    componentLinearAccScale=<1 1 1>
    componentLinearVelScale=<1 1 1>
    gravityOverride {z=-980}
    physicsBodyDefinitions [{
      boundBone {boneName=dyn_strapB_04}
      boxExtents=<4.5 4.5 4.5>
      constraintSetup {angularLimitsMax=<5 25 25>, angularLimitsMin=<-5 -25 -25>, angularTarget=<1 0 0>}
      localJointOffset=<0.2 0 0>
     }]
    planarLimits [{drivingBone {boneName=spine_04}, planeTransform {location=<7 -12 0>, rotation=<rot 0 40 -90>}}]
    relativeSpaceBone {boneName=head}
    sphericalLimits [
     {limitRadius=14, drivingBone {boneName=head}, sphereLocalOffset=<0 5 0>}
     {limitRadius=11, drivingBone {boneName=clavicle_l}, sphereLocalOffset=<5 0 0>}
     {limitRadius=11, drivingBone {boneName=clavicle_l}, sphereLocalOffset=<15 0 0>}
     {limitRadius=11, drivingBone {boneName=clavicle_r}, sphereLocalOffset=<-5 0 0>}
     {limitRadius=11, drivingBone {boneName=clavicle_r}, sphereLocalOffset=<-15 0 0>}
     {limitRadius=11, drivingBone {boneName=spine_05}, sphereLocalOffset=<7 0 0>}
    ]
   }
   n23 = Anim Dynamics - Bone: dyn_strapB_05!(ComponentPose=@n22, Alpha=1, GravityOverride:Vector="0,0,-980", ExternalForce="0,0,0") -> (Pose) {
    angularDampingOverride=0.999
    angularSpringConstant=50
    bAngularSpring=true
    bOverrideAngularDamping=true
    bUseGravityOverride=true
    bUseSphericalLimits=true
    numSolverIterationsPostUpdate=3
    numSolverIterationsPreUpdate=6
    simulationSpace=RootRelative
    alphaBoolBlend {blendOption=Linear}
    boundBone {boneName=dyn_strapB_05}
    componentAppliedLinearAccClamp=<10000 10000 10000>
    componentLinearAccScale=<1 1 1>
    componentLinearVelScale=<1 1 1>
    gravityOverride {z=-980}
    physicsBodyDefinitions [{
      boundBone {boneName=dyn_strapB_05}
      boxExtents=<4.5 4.5 4.5>
      constraintSetup {angularLimitsMax=<5 25 25>, angularLimitsMin=<-5 -25 -25>, angularTarget=<1 0 0>}
      localJointOffset=<0.2 0 0>
     }]
    planarLimits [{drivingBone {boneName=spine_04}, planeTransform {location=<7 -12 0>, rotation=<rot 0 40 -90>}}]
    relativeSpaceBone {boneName=head}
    sphericalLimits [
     {limitRadius=14, drivingBone {boneName=head}, sphereLocalOffset=<0 5 0>}
     {limitRadius=11, drivingBone {boneName=clavicle_l}, sphereLocalOffset=<5 0 0>}
     {limitRadius=11, drivingBone {boneName=clavicle_l}, sphereLocalOffset=<15 0 0>}
     {limitRadius=11, drivingBone {boneName=clavicle_r}, sphereLocalOffset=<-5 0 0>}
     {limitRadius=11, drivingBone {boneName=clavicle_r}, sphereLocalOffset=<-15 0 0>}
     {limitRadius=11, drivingBone {boneName=spine_05}, sphereLocalOffset=<7 0 0>}
    ]
   }
   n24 = RigidBody!(ComponentPose=Local To Component(LocalPose=Use cached pose 'AnimCache'()), Alpha, OverrideWorldGravity:Vector="0,0,0", ExternalForce) -> (Pose) {
    bClampLinearTranslationLimitToRefPose=true
    bOverrideWorldGravity=true
    overridePhysicsAsset=&/M_MED_TeriyakiWarrior_FaceAcc_Bandana_Physics_upd2
    simulationSpace=BaseBoneSpace
    alphaBoolBlend {blendOption=Linear}
    baseBoneRef {boneName=head}
    simSpaceSettings {maxAngularAcceleration=30, maxAngularVelocity=35, maxLinearAcceleration=2200, maxLinearVelocity=2200, worldAlpha=1}
   }
   n25 = Blend Poses by int!(
    BlendPose_0:Pose Link Structure=Component To Local(ComponentPose=@n24)
    BlendPose_1:Pose Link Structure=Component To Local(ComponentPose=@n23)
    BlendPose_2:Pose Link Structure=Component To Local(ComponentPose=@n13)
    BlendTime_0:float=0.1
    BlendTime_1:float=0.1
    BlendTime_2:float=0.1
    ActiveChildIndex:int=0
   ) -> (Pose:Pose Link Structure) {
    is=title AnimGraphNode_BlendListByInt
    blendType=Linear
    blendPose [{linkId=-1, sourceLinkId=-1}, {linkId=-1, sourceLinkId=-1}, {linkId=-1, sourceLinkId=-1}]
    blendTime [0.1,0.1,0.1]
   }
   Output Pose(Result=@n25) {layerGroup=None}
   n26 = Layered blend per bone(
    BasePose=Copy Pose From Mesh() {bUseAttachedParent=true}
    BlendPoses_0=Copy Pose From Mesh(SourceMeshComponent=Get Body() -> (Body:Skeletal Mesh Component Budgeted Object Reference))
   ) {layerSetup [{branchFilters [
       {blendDepth=2, boneName=clavicle_r}
       {blendDepth=2, boneName=clavicle_l}
       {blendDepth=2, boneName=upperarm_l}
       {blendDepth=2, boneName=upperarm_r}
      ]}]}
   n27 = "Constraint - dyn_strap_parent (PPP)"!(
    ComponentPose=Local To Component(LocalPose=@n26)
    Alpha=1
    ConstraintWeights_0=0.0863
    ConstraintWeights_1:float=0.2731
    ConstraintWeights_2:float=0.6406
   ) -> (Pose) {
    alphaBoolBlend {blendOption=Linear}
    boneToModify {boneName=dyn_strap_parent}
    constraintSetup [{targetBone {boneName=neck_01}}, {targetBone {boneName=neck_02}}, {targetBone {boneName=head}}]
    constraintWeights [1,1,1]
   }
   Save cached pose 'AnimCache'(Pose=Component To Local(ComponentPose=@n27)) {cacheName=AnimCache}
  ]}
 EventGraph {nodes [
   Event BlueprintUpdateAnimation()
   > n1 = Sequence()
   > n2 = Gravity Override State Function DEPRECATED(
    CharacterPartAnimInstance=Self-Reference()
    GameTimeInSeconds=GetGameTimeInSeconds()
    GravityOverrideParams=Get Gravity Override Params()
    RBAN Params=Get RBAN Params()
   )
   > Set GravityOverride(GravityOverride=@n2)
   n3 = Branch(execute=@n1.then_1, Condition=Get bRigidBodyEnabled())
   > Set SimType(SimType:int=0) -> (Output_Get:int=0)
   Set SimType(execute=[@n3.else, @n16.else], SimType=2) -> (Output_Get=0)
   n4 = Event OnCharacterCustomizationLoaded()
   > Set Body(Body:Skeletal Mesh Component Budgeted Object Reference=Get Array Item with EFort Custom Part Type(Array=@n4.CustomMeshParts)) -> (Output_Get:Skeletal Mesh Component Budgeted Object Reference)
   n5 = Sequence()
   n6 = Get PawnSpeed2D()
   > n7 = Branch(Condition="float <= float"!(A:double=MapRangeClamped(Value=@n6, InRangeA=0, InRangeB=150), B:double=0) -> (ReturnValue:bool) {is=title K2Node_CallFunction})
   > Set straps_vector(straps_vector:Vector="10,800,800") -> (Output_Get:Vector)
   > Set standing_still(standing_still:bool=true) -> (Output_Get:bool)
   > Set strap_rot_alpha(strap_rot_alpha:double=0) -> (Output_Get:double)
   > Set straps_standing_alpha(straps_standing_alpha:double=0.9) -> (Output_Get:double)
   > Set straps_vector(straps_vector="0,0,1")
   Set straps_vector(execute=@n7.else, straps_vector="1,0.05,0.05")
   > Set standing_still(standing_still=false)
   > Set straps_standing_alpha(straps_standing_alpha=0.9)
   > Set straps_stand_rot(straps_stand_rot:double=0) -> (Output_Get:double)
   > Set strap_rot_alpha(strap_rot_alpha=0)
   > n8 = Branch(Condition=Get isMovingBackward() -> (isMovingBackward:bool))
   n9 = Get player_pawn() -> (player_pawn:Fort Player Pawn Athena Object Reference)
   n10 = Get Actor Rotation!(self:Actor Object Reference=@n9) -> (ReturnValue:Rotator) {is=title K2Node_CallFunction}
   n11 = BreakVector(InVec=vector * vector(
     A:Vector=Get Rotation X Vector!(InRot:Rotator=@n10) -> (ReturnValue:Vector) {is=title K2Node_CallFunction}
     B:Vector=GetVelocity!(self:Actor Object Reference=@n9) -> (ReturnValue:Vector)
    ) -> (ReturnValue:Vector) {is=title K2Node_CommutativeAssociativeBinaryOperator})
   n12 = float + float(A=@n11, B=@n11.Y) {is=title K2Node_CommutativeAssociativeBinaryOperator}
   n13 = Branch(execute=@n8.else, Condition="float < float"!(A:double=@n12, B:double=415) -> (ReturnValue:bool) {is=title K2Node_CallFunction})
   > Set straps_vector(straps_vector="0.25,2.5,0.6")
   > n14 = Branch(Condition=Get bIsCrouching())
   > Set strap_rot_alpha(strap_rot_alpha=0)
   Set strap_rot_alpha(execute=@n14.else, strap_rot_alpha=0.6)
   n15 = Branch(execute=@n13.else, Condition=Get bIsJumping())
   > Set straps_vector(straps_vector="0, 0,0.65")
   > Set strap_rot_alpha(strap_rot_alpha=0.15)
   Set straps_vector(execute=@n15.else, straps_vector="0,1,0.65")
   > Set strap_rot_alpha(strap_rot_alpha=1)
   Set pawn_speed(execute=@n5.then_1, pawn_speed:double=@n12) -> (Output_Get:double)
   > Set isMovingBackward(isMovingBackward:bool="float < float"(A=@n12, B=0)) -> (Output_Get:bool)
   n16 = Branch(Condition="integer <= integer"(A=Get CurrentLOD(), B=1))
   > Set SimType(SimType=1) -> (Output_Get=0)
   n17 = TryGetPawnOwner()
   Is Valid(InputObject=@n17)
   > n18 = Cast To FortPlayerPawnAthena(Object=@n17)
   > Set player_pawn(player_pawn:Fort Player Pawn Athena Object Reference=@n18) -> (Output_Get:Fort Player Pawn Athena Object Reference)
   Break Fort Anim Input Skydiving External Force(FortAnimInput_SkydivingExternalForce=Get SkydivingExternalForceInput()) -> (
    bUseSkydivingVectorForce=~
    SkydivingVectorForce=~
    HeadToPelvisDirection=~
    FloatingMultiplier=~
    FloatingAdditive=~
    DivingMultiplier=~
    DivingAdditive=~
    ParachutingMultiplier=~
    ParachutingAdditive=~
   )
   Event BlueprintInitializeAnimation()
   Is Valid(InputObject=Get player_pawn())
  ]}
]
variables [
 Body {type=SkeletalMeshComponentBudgeted}
 Gravity Override Params {
  type=GravityOverrideParamsStruct
  default {
   skydive_FloatingDown {gravityOverride=<0 -700 -980>}
   skydive_GoingUp {gravityOverride=<0 -300 -1500>}
   skydive_Parachuting {gravityOverride=<0 -200 -980>}
  }
 }
 GravityOverride {type=Vector}
 isMovingBackward {type=boolean}
 pawn_speed {type=number}
 player_pawn {type=FortPlayerPawnAthena}
 RBAN Params {
  type=RBANParamsStruct
  default {
   dBNO {simSpaceSettings {
     maxAngularAcceleration=1000
     maxAngularVelocity=3
     maxLinearAcceleration=300
     maxLinearVelocity=700
     velocityScaleZ=0
     externalLinearDragV=<0 0.5 0>
    }}
   dBNO_Carried {simSpaceSettings {
     maxAngularAcceleration=1000
     maxAngularVelocity=3
     maxLinearAcceleration=300
     maxLinearVelocity=700
     velocityScaleZ=0
     externalLinearDragV=<0 0.5 0>
    }}
   emoteOrMelee {simSpaceSettings {
     maxAngularAcceleration=1000
     maxAngularVelocity=3
     maxLinearAcceleration=300
     maxLinearVelocity=700
     velocityScaleZ=0
     externalLinearDragV=<0 0.5 0>
    }}
   front_End {simSpaceSettings {
     maxAngularAcceleration=1000
     maxAngularVelocity=3
     maxLinearAcceleration=300
     maxLinearVelocity=700
     velocityScaleZ=0
     externalLinearDragV=<0 0.5 0>
    }}
   onGround_Crouch_Moving {simSpaceSettings {
     maxAngularAcceleration=1000
     maxAngularVelocity=3
     maxLinearAcceleration=300
     maxLinearVelocity=700
     velocityScaleZ=0
     externalLinearDragV=<0 0.5 0>
    }}
   onGround_Falling {simSpaceSettings {
     maxAngularAcceleration=1000
     maxAngularVelocity=3
     maxLinearAcceleration=300
     maxLinearVelocity=700
     velocityScaleZ=0
     externalLinearDragV=<0 0.5 0>
    }}
   onGround_Moving {simSpaceSettings {
     maxAngularAcceleration=1000
     maxAngularVelocity=3
     maxLinearAcceleration=300
     maxLinearVelocity=700
     velocityScaleZ=0
     externalLinearDragV=<0 0.5 0>
    }}
   onGround_Standing {simSpaceSettings {
     maxAngularAcceleration=1000
     maxAngularVelocity=3
     maxLinearAcceleration=300
     maxLinearVelocity=700
     velocityScaleZ=0
     externalLinearDragV=<0 0.5 0>
    }}
   rail_Grinding {simSpaceSettings {
     maxAngularAcceleration=1000
     maxAngularVelocity=3
     maxLinearAcceleration=300
     maxLinearVelocity=700
     velocityScaleZ=0
     worldAlpha=0.1
     externalLinearDragV=<0 0.5 0>
    }}
   riding_Moving {simSpaceSettings {
     maxAngularAcceleration=1000
     maxAngularVelocity=3
     maxLinearAcceleration=300
     maxLinearVelocity=700
     velocityScaleZ=0
     externalLinearDragV=<0 0.5 0>
    }}
   riding_Moving_Targeting {simSpaceSettings {
     maxAngularAcceleration=1000
     maxAngularVelocity=3
     maxLinearAcceleration=300
     maxLinearVelocity=700
     velocityScaleZ=0
     externalLinearDragV=<0 0.5 0>
    }}
   riding_Standing {simSpaceSettings {
     maxAngularAcceleration=1000
     maxAngularVelocity=3
     maxLinearAcceleration=300
     maxLinearVelocity=700
     velocityScaleZ=0
     externalLinearDragV=<0 0.5 0>
    }}
   skydive_DivingDown {simSpaceSettings {
     maxAngularAcceleration=1000
     maxAngularVelocity=3
     maxLinearAcceleration=300
     maxLinearVelocity=700
     velocityScaleZ=0
     externalLinearDragV=<0 0.5 0>
    }}
   skydive_FloatingDown {
    componentLinearAccScale=<0 0.5 0>
    componentLinearVelScale=<0 0.5 0>
    simSpaceSettings {
     maxAngularAcceleration=1000
     maxAngularVelocity=3
     maxLinearAcceleration=300
     maxLinearVelocity=700
     velocityScaleZ=0
     externalLinearDragV=<0 0.5 0>
    }
   }
   skydive_GoingUp {simSpaceSettings {
     maxAngularAcceleration=1000
     maxAngularVelocity=3
     maxLinearAcceleration=300
     maxLinearVelocity=700
     velocityScaleZ=0
     externalLinearDragV=<0 0.5 0>
    }}
   skydive_Parachuting {
    componentLinearAccScale=<0 0.1 0.3>
    componentLinearVelScale=<0 0.1 0.3>
    simSpaceSettings {
     maxAngularAcceleration=1000
     maxAngularVelocity=3
     maxLinearAcceleration=300
     maxLinearVelocity=700
     velocityScaleZ=0
     externalLinearDragV=<0 0.5 0>
    }
   }
   skydive_Surfing {simSpaceSettings {
     maxAngularAcceleration=1000
     maxAngularVelocity=3
     maxLinearAcceleration=300
     maxLinearVelocity=700
     velocityScaleZ=0
     externalLinearDragV=<0 0.5 0>
    }}
   swimming {simSpaceSettings {
     maxAngularAcceleration=1000
     maxAngularVelocity=3
     maxLinearAcceleration=300
     maxLinearVelocity=700
     velocityScaleZ=0
     externalLinearDragV=<0 0.5 0>
    }}
   tactical_Sprint {simSpaceSettings {
     maxAngularAcceleration=1000
     maxAngularVelocity=3
     maxLinearAcceleration=300
     maxLinearVelocity=700
     velocityScaleZ=0
     externalLinearDragV=<0 0.5 0>
    }}
   ziplining {simSpaceSettings {
     maxAngularAcceleration=1000
     maxAngularVelocity=3
     maxLinearAcceleration=300
     maxLinearVelocity=700
     velocityScaleZ=0
     externalLinearDragV=<0 0.5 0>
    }}
  }
 }
 SimType {type=integer}
 standing_still {type=boolean}
 strap_rot_alpha {type=number}
 straps_force_vector {type=Vector}
 straps_stand_rot {type=number}
 straps_standing_alpha {type=number, default=0.9}
 straps_vector {type=Vector}
]
