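# Physics asset for the Teriyaki Warrior hat and face accessory, covering jointed motion of the left and right helmet straps and a lens cap.
# Each strap is a five-link, mostly box chain from its base body, with twist locked and limited swing.
# The lens cap hangs on tightly limited capsule links, ending in a box that swings and twists freely.
# Head and neck spheres and capsules provide body volumes, with no constraints to the chains.
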PhysicsAsset /BRCosmetics/Characters/Player/Male/Medium/Bodies/M_MED_Teriyaki_Warrior/Meshes/Parts/M_MED_Teriyaki_Warrior_Hat_FaceAcc_Physics
bodies {
 dyn_helmStrap_01_Base_l [{shapeType=Capsule, length=0.505, radius=0.505, rotation=<rot 0 -90 0>}]
 dyn_helmStrap_01_Base_r [{shapeType=Box, extentX=1.01, extentY=1.01, extentZ=1.01, rotation=<rot 0 90 0>}]
 dyn_helmStrap_01_l [{shapeType=Box, extentX=1.01, extentY=5.0974, extentZ=13.0406, center=<-1.8536 0.0577 -0.0164>, rotation=<rot 7.8913 101.6775 179.8449>}]
 dyn_helmStrap_01_r [{shapeType=Box, extentX=1.0094, extentY=5.0974, extentZ=12.8943, center=<1.5152 -0.0738 0.0245>, rotation=<rot 8.3118 94.7732 -1.0817>}]
 dyn_helmStrap_02_l [{shapeType=Box, extentX=1.0315, extentY=4.733, extentZ=11.0104, center=<-1.7244 0.0001 0>, rotation=<rot 9.1676 89.9998 179.9999>}]
 dyn_helmStrap_02_r [{shapeType=Box, extentX=0.9728, extentY=4.1448, extentZ=12.8787, center=<1.7244 0.0981 0.0158>, rotation=<rot 9.1676 89.9998 -0.0001>}]
 dyn_helmStrap_03_l [{shapeType=Box, extentX=0.9509, extentY=4.3746, extentZ=9.0206, center=<-1.6924 -0.1007 -0.0129>, rotation=<rot 7.2874 89.9998 -180>}]
 dyn_helmStrap_03_r [{shapeType=Box, extentX=0.9509, extentY=4.1433, extentZ=9.3205, center=<1.6924 0.1007 0.0129>, rotation=<rot 7.2874 89.9998 0>}]
 dyn_helmStrap_04_l [{shapeType=Box, extentX=1.0472, extentY=1.6342, extentZ=3.3951, center=<-0.6574 -0.0852 0>, rotation=<rot 0 90 180>}]
 dyn_helmStrap_04_r [{shapeType=Box, extentX=1.0472, extentY=1.6342, extentZ=3.3951, center=<0.6574 0.0852 0>, rotation=<rot 0 90 0>}]
 dyn_helmStrap_05_l [{shapeType=Box, extentX=1.5378, extentY=5.2713, extentZ=2.7178, center=<-1.5455 0.0001 0>, rotation=<rot 0 90 180>}]
 dyn_helmStrap_05_r [{shapeType=Box, extentX=1.5378, extentY=5.2713, extentZ=2.7178, center=<1.5455 -0.0001 0>, rotation=<rot 0 90 0>}]
 dyn_lensCap_01 [{shapeType=Capsule, length=1.1186, radius=0.505, center=<-0.6926 0.0001 0>, rotation=<rot 90 0 0>}]
 dyn_lensCap_01_Base [{shapeType=Capsule, length=13.0755, radius=3.7196, center=<0.8899 -0.559 -3.9972>, rotation=<rot -11.7093 -92.7744 22.1946>}]
 dyn_lensCap_02 [{shapeType=Capsule, length=0.6731, radius=0.5593, center=<-0.6792 0 0>, rotation=<rot 90 0 0>}]
 dyn_lensCap_03 [{shapeType=Capsule, length=0.346, radius=0.5001, center=<-0.3853 0 0>, rotation=<rot 90 0 0>}]
 dyn_lensCap_04 [{shapeType=Box, extentX=1.5186, extentY=6.2448, extentZ=6.457, center=<-3.4445 -0.2985 -0.0135>, rotation=<rot 0 92.7238 -44.3659>}]
 head [
  {shapeType=Sphere, radius=3.5201, shapeName=head_sphere, center=<6.6091 11.521 -12.5956>}
  {shapeType=Sphere, radius=3.5201, shapeName=head_sphere, center=<6.3561 10.5681 11.2087>}
  {shapeType=Capsule, length=9.2516, radius=3.15, shapeName=head_capsule, center=<11.2698 0 14.1995>, rotation=<rot 0 -0.0001 -90>}
  {shapeType=Capsule, length=10.012, radius=3.6673, shapeName=head_capsule, center=<11.5489 0.3814 -15.3024>, rotation=<rot 0 -10.0001 -90>}
  {shapeType=Capsule, length=17.1617, radius=12.9305, shapeName=head_capsule, center=<-0.3951 1.93 0>, rotation=<rot 0 -10.0002 90>}
 ]
 neck_01 [{shapeType=Capsule, length=5.2418, radius=12.8478, shapeName=neck_01_capsule, center=<0 4.9771 0>, rotation=<rot 80 89.9996 -90>}]
}
constraints [
 {bone1Name=dyn_lensCap_01, bone2Name=dyn_lensCap_01_Base, swing1LimitDegrees=12, swing2LimitDegrees=12, twistLimitDegrees=12}
 {bone1Name=dyn_lensCap_02, bone2Name=dyn_lensCap_01, swing1LimitDegrees=12, swing2LimitDegrees=12, twistLimitDegrees=12}
 {bone1Name=dyn_lensCap_03, bone2Name=dyn_lensCap_02, swing1LimitDegrees=12, swing2LimitDegrees=12, twistLimitDegrees=12}
 {
  bone1Name=dyn_lensCap_04
  bone2Name=dyn_lensCap_03
  swing1LimitDegrees=45
  swing1Motion=Free
  swing2LimitDegrees=45
  swing2Motion=Free
  twistLimitDegrees=45
  twistMotion=Free
 }
 {
  bone1Name=dyn_helmStrap_01_r
  bone2Name=dyn_helmStrap_01_Base_r
  swing1LimitDegrees=30
  swing2LimitDegrees=10
  twistLimitDegrees=45
  twistMotion=Locked
 }
 {
  bone1Name=dyn_helmStrap_02_r
  bone2Name=dyn_helmStrap_01_r
  swing1LimitDegrees=25
  swing2LimitDegrees=10
  twistLimitDegrees=45
  twistMotion=Locked
 }
 {
  bone1Name=dyn_helmStrap_03_r
  bone2Name=dyn_helmStrap_02_r
  swing1LimitDegrees=20
  swing2LimitDegrees=10
  twistLimitDegrees=45
  twistMotion=Locked
 }
 {
  bone1Name=dyn_helmStrap_04_r
  bone2Name=dyn_helmStrap_03_r
  swing1LimitDegrees=10
  swing2LimitDegrees=10
  twistLimitDegrees=45
  twistMotion=Locked
 }
 {
  bone1Name=dyn_helmStrap_05_r
  bone2Name=dyn_helmStrap_04_r
  swing1LimitDegrees=20
  swing2LimitDegrees=20
  twistLimitDegrees=45
  twistMotion=Locked
 }
 {
  bone1Name=dyn_helmStrap_02_l
  bone2Name=dyn_helmStrap_01_l
  swing1LimitDegrees=30
  swing2LimitDegrees=10
  twistLimitDegrees=45
  twistMotion=Locked
 }
 {
  bone1Name=dyn_helmStrap_03_l
  bone2Name=dyn_helmStrap_02_l
  swing1LimitDegrees=20
  swing2LimitDegrees=10
  twistLimitDegrees=45
  twistMotion=Locked
 }
 {
  bone1Name=dyn_helmStrap_04_l
  bone2Name=dyn_helmStrap_03_l
  swing1LimitDegrees=10
  swing2LimitDegrees=10
  twistLimitDegrees=45
  twistMotion=Locked
 }
 {
  bone1Name=dyn_helmStrap_05_l
  bone2Name=dyn_helmStrap_04_l
  swing1LimitDegrees=20
  swing2LimitDegrees=20
  twistLimitDegrees=45
  twistMotion=Locked
 }
 {
  bone1Name=dyn_helmStrap_01_l
  bone2Name=dyn_helmStrap_01_Base_l
  swing1LimitDegrees=30
  swing2LimitDegrees=10
  twistLimitDegrees=45
  twistMotion=Locked
 }
]
